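# Prototype Niagara skydiving contrail "EnergyNet": an audio-reactive effect attached to a character's skeleton.
# Mesh boxes and a ring follow the body, with sprites, trail meshes and rising mesh particles on bones, plus bone-flow ribbons between limb and shoulder bones.
# A system audio value drives brightness, noise, spawn rate, trail scale and ribbon bulge, while dive state adjusts orientation, offsets and scale.
# Separate front-end box and ring emitters are disabled.
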NiagaraSystem /BRCosmetics/Athena/Effects/Particles/Skydiving/NS_Contrail_EnergyNet_Proto/NS_Contrail_EnergyNet_Proto
dependencies {
 usedDataInterfaces [
  &NiagaraDataInterfaceArrayFloat3
  &NiagaraDataInterfaceCurve
  &NiagaraDataInterfaceColorCurve
  &NiagaraDataInterfaceArrayFloat
 ]
 usedDynamicInputs [
  &RemapRange
  &MultiplyLinearColor
  &MakeVector4_FromFloat
 ]
 usedModules [
  &SystemState
  &self:SystemSpawnScript.SystemScriptSource.SystemScriptGraph.NiagaraNodeAssignment_0.SetVariables_#1
  &self:SystemSpawnScript.SystemScriptSource.SystemScriptGraph.NiagaraNodeAssignment_1.SetVariables_#2
  &self:SystemSpawnScript.SystemScriptSource.SystemScriptGraph.NiagaraNodeAssignment_2.SetVariables_#3
 ]
}
emitters {
 Debug {
  bEnabled=false
  emitterUpdateScript [
   &EmitterState(Life Cycle Mode=::NewEnumerator1, Loop Behavior=::NewEnumerator1, Loop Duration Mode=::NewEnumerator1, UseLoopDelay=false)
   &SpawnBurst_Instantaneous(Use Spawn Probability=false, UseLoopCountLimit=false)
  ]
  particleSpawnScript [&InitializeParticle(
    Color Mode=::NewEnumerator0
    Lifetime=5
    Sprite Size Mode=::NewEnumerator3
    Uniform Sprite Size=100
    UsePositionOffset=false
    Write Material Random=true
   )]
  particleUpdateScript [
   &ParticleState(Kill Particles When Lifetime Has Elapsed=false, Let Infinitely Lived Particles Die When Emitter Deactivates=true)
   &ScaleSpriteSize(
    Scale Sprite Size Mode=::NewEnumerator0
    Uniform Scale Factor=&Add_Float(
     A=1
     B=&Multiply_Float(
      A=&MakeFloatFromLinearColor(LinearColor=$NPC.NPC_Audio_Reactive_Parameters.Spectrum_Music, Channel:ENiagaraLinearColor_Channels=::NewEnumerator0)
      B=1
     )
    )
   )
   &SkeletalMeshLocation(Bone Sampling Mode=::NewEnumerator4, Inherit Velocity=false, MaskSpawnGroup=false, Skeletal Mesh=$User.Skeletal Mesh Root)
  ]
  rendererClasses [&NiagaraSpriteRendererProperties]
  renderers [{rendererClass=&NiagaraSpriteRendererProperties}]
 }
 Mesh_Box {
  emitterUpdateScript [
   &EmitterState(
    Inactive Response=::NewEnumerator1
    Life Cycle Mode=::NewEnumerator1
    Loop Behavior=::NewEnumerator1
    Loop Duration Mode=::NewEnumerator1
    UseLoopDelay=false
   )
   &SpawnBurst_Instantaneous(
    Spawn Count=&MakeCustomIntFromBool(Bool=$User.bIsFrontEndPreview, True INT=0, False INT=45)
    Use Spawn Probability=false
   )
  ]
  particleSpawnScript [
   &InitializeParticle(Color Mode=::NewEnumerator0, Mesh Scale=<1 0.52 0.25>, Mesh Scale Mode=::NewEnumerator4, UsePositionOffset=false, Write Material Random=true)
   &FN_SetSpawnGroup(Normalize Spawn Group=true, Particles Per Group=45, Pattern=::NewEnumerator0)
   &TorusLocation(
    Align Torus With Axis=true
    Handle Radius=0
    Large Radius=55
    MaskSpawnGroup=false
    Normalized Angle Around Torus Axis:NiagaraFloat=$Particles.Normalized Spawn Group
    Torus Distribution Mode:ENiagaraTorusDistributionMode=::NewEnumerator1
   ) {enabled=false}
   &ShapeLocation(
    Cylinder Distribution=::NewEnumerator2
    Cylinder Height=0
    Cylinder Radius=45
    MaskSpawnGroup=false
    Shape Primitive=::NewEnumerator1
    U Position=$Particles.Normalized Spawn Group
    UseOwnerScale=true
   )
   &InitialMeshOrientation(Use Orientation Vector=true, Use Rotation Vector=true)
  ]
  particleUpdateScript [
   /Game/Effects/Niagara/Modules/Skydiving/FN_DivingAttributes/FN_DivingAttributes_v2(
    Accumulate Dive=1
    Accumulate Rest=0
    Bone Index=0
    SkeletalMesh=DataInterface() {propertyValues {
      DefaultMesh=null
      ExcludeBoneName=root
      OptionalComponentClass=null
      PreviewMesh=&M_Med_Soldier_04_Preview
      SkinningMode=SkinOnTheFly
      SoftSourceActor=null
      SourceMode=Default
      UvSetIndex=0
      WholeMeshLOD=-1
      bExcludeBone=false
      bReadDeformedGeometry=true
      bRequireCurrentFrameData=true
      ComponentTags []
      FilteredBones [spine_02]
      FilteredSockets []
      MeshUserParameter {parameter {name=None, typeDefHandle {registeredTypeIndex=1}, varData []}}
      SamplingRegions []
     }}
   )
   &ParticleState(
    Kill Particles When Lifetime Has Elapsed=false
    Let Infinitely Lived Particles Die When Emitter Deactivates=true
    Loop Particles Lifetime=true
   )
   &SkeletalMeshLocation(
    Bone Sampling Mode=::NewEnumerator4
    Inherit Velocity=false
    MaskSpawnGroup=false
    Sampled Position Offset:NiagaraPosition=$Particles.Initial.Position
    Skeletal Mesh=$User.Skeletal Mesh Root
   )
   /SetVariables(Particles.MeshOrientation=&MultiplyQuaternion(
     Quaternion A=&FindQuatBetweenTwoVectors(
      From Vector=<1 0 0>
      To Vector=&NormalizeVector(Vector To Normalize=&Subtract_Vector(
        A=&MakeCustomVectorFromBool(
         Bool=&SetBoolByFloatComparison(A=$Particles.Normalized Spawn Group, B=0.5, Comparison Type=::NewEnumerator4)
         True Vector=&Subtract_Vector(A=Linked() {linkedVariable {name=Particles.Position, type=NiagaraPosition}}, B=<0 1 0>)
         False Vector=Linked() {linkedVariable {name=Particles.Position, type=NiagaraPosition}}
        )
        B=$Output.SkeletalMeshLocation.SampledPosition
       ))
     )
     Quaternion B=$Particles.Initial.MeshOrientation
    )) {isSetParameters=true, module=SetVariables_#4}
   &UpdateMeshOrientation(
    Rotation Coordinate Space=::NewEnumerator3
    Rotation Vector=&Lerp_Vector(
     A=&FN_FrontEndVector3Modifier(Gameplay=<0 0 0>, Front End=<0 0 0>)
     B=<0 -5 0>
     Alpha=$Particles.Dive
    )
   )
   &self:NoiseScale(Noise Speed=1.5, Scale Size:NiagaraFloat=&Lerp_Float(A=0.12, B=1.2, Alpha=$System.Audio NPC 2)) {module=NoiseScale}
   /SetVariables(Particles.Position=&Add_Vector(
     A=Linked() {linkedVariable {name=Particles.Position, type=NiagaraPosition}}
     B=&Multiply_VectorByFloat(
      Vector=$Particles.Initial.ShapeLocation.ShapeNormal
      Float=&Multiply_Float(A=&Power(a=$System.Audio NPC 2, b=1), B=-70)
     )
    )) {enabled=false, isSetParameters=true, module=SetVariables_#5}
   &Color(
    Color=&ColorFromCurve(
     ColorCurve=DataInterface() {propertyValues {
       AlphaCurve {keys [<at 0 1 RCIM_Cubic>]}
       BlueCurve {keys [<at 0.014 0.15 RCIM_Cubic>,<at 0.0758 0.2>,<at 0.1399 1>,<at 0.3533 1>,<at 0.4057 1>,<at 0.4431 1>,<at 0.4909 0.15>,<at 0.5363 0.15>,<at 0.5876 0.2>,<at 0.6366 1>,<at 0.888 1>,<at 0.9174 1>,<at 0.9527 1>,<at 1 0.15>]}
       GreenCurve {keys [<at 0.014 0.15 RCIM_Cubic>,<at 0.0758 0.5>,<at 0.1399 1>,<at 0.3533 0.6667>,<at 0.4057 0.2>,<at 0.4431 0.2>,<at 0.4909 0.15>,<at 0.5363 0.15>,<at 0.5876 0.5>,<at 0.6366 1>,<at 0.888 0.6667>,<at 0.9174 0.2>,<at 0.9527 0.2>,<at 1 0.15>]}
       RedCurve {keys [<at 0.014 0.3 RCIM_Cubic>,<at 0.0758 0.05>,<at 0.1399 0>,<at 0.3533 0>,<at 0.4057 0.3333>,<at 0.4431 0.2>,<at 0.4909 0.3>,<at 0.5363 0.3>,<at 0.5876 0.05>,<at 0.6366 0>,<at 0.888 0>,<at 0.9174 0.3333>,<at 0.9527 0.2>,<at 1 0.3>]}
      }}
     CurveIndex=$Particles.Normalized Spawn Group
    )
    Scale Color=&Add_Vector(
     A=<0.1 0.1 0.1>
     B=&Multiply_VectorByFloat(Vector=<1 1 1>, Float=&Power(a=$System.Audio NPC 2, b=2))
    )
   )
   &ScaleColor(
    Color Value To Scale=$Particles.Color
    Linear Color Curve=DataInterface() {propertyValues {
      AlphaCurve {keys [<at 0 1>]}
      BlueCurve {keys [<at 0 0.444>,<at 0.2641 0.96>,<at 0.4666 0.36>,<at 0.7161 1.2>,<at 1 0.444>]}
      GreenCurve {keys [<at 0 0.36>,<at 0.2641 1.2>,<at 0.4666 0.5467>,<at 0.7161 1.0133>,<at 1 0.36>]}
      RedCurve {keys [<at 0 1.2>,<at 0.2641 0.12>,<at 0.4666 1.2>,<at 0.7161 0.36>,<at 1 1.2>]}
     }}
    Scale Mode=::NewEnumerator2
   )
  ]
  rendererClasses [&NiagaraMeshRendererProperties]
  renderers [{rendererClass=&NiagaraMeshRendererProperties, rendererIndex=0}]
 }
 Mesh_Box_FE {
  bEnabled=false
  emitterUpdateScript [
   &EmitterState(Life Cycle Mode=::NewEnumerator1, Loop Duration=20, UseLoopDelay=false)
   &SpawnBurst_Instantaneous(
    Spawn Count=&MakeCustomIntFromBool(Bool=$User.bIsFrontEndPreview, True INT=45, False INT=0)
    Use Spawn Probability=false
   )
  ]
  particleSpawnScript [
   &InitializeParticle(
    Color Mode=::NewEnumerator0
    Lifetime=20
    Mesh Scale=<1.2 0.6 0.3>
    Mesh Scale Mode=::NewEnumerator4
    UsePositionOffset=false
    Write Material Random=true
   )
   &FN_SetSpawnGroup(Normalize Spawn Group=true, Particles Per Group=45, Pattern=::NewEnumerator0)
   &TorusLocation(
    Align Torus With Axis=true
    Handle Radius=0
    Large Radius=60
    MaskSpawnGroup=false
    Normalized Angle Around Torus Axis:NiagaraFloat=$Particles.Normalized Spawn Group
    Torus Distribution Mode:ENiagaraTorusDistributionMode=::NewEnumerator1
   )
   &InitialMeshOrientation(Use Orientation Vector=true, Use Rotation Vector=true)
  ]
  particleUpdateScript [
   /SetVariables(Particles.FE DivingCurve:NiagaraFloat=&FloatFromCurve(
     FloatCurve=DataInterface() {propertyValues {Curve {keys [
         <at 0 0 RCIM_Cubic>
         {arriveTangent=2.2332, leaveTangent=2.2332, time=0.02, value=1}
         {arriveTangent=-1.0204, interpMode=RCIM_Cubic, leaveTangent=-1.0204, time=0.5, value=1}
         <at 0.52 0>
         <at 1 0>
        ]}}}
     CurveIndex=$Emitter.NormalizedLoopAge
    )) {isSetParameters=true}
   &FN_DivingAttributes_v2(
    Accumulate Dive=1
    Accumulate Rest=0
    Bone Index=0
    SkeletalMesh=DataInterface() {propertyValues {
      DefaultMesh=null
      ExcludeBoneName=root
      OptionalComponentClass=null
      PreviewMesh=&M_Med_Soldier_04_Preview
      SkinningMode=SkinOnTheFly
      SoftSourceActor=null
      SourceMode=Default
      UvSetIndex=0
      WholeMeshLOD=-1
      bExcludeBone=false
      bReadDeformedGeometry=true
      bRequireCurrentFrameData=true
      ComponentTags []
      FilteredBones [spine_02]
      FilteredSockets []
      MeshUserParameter {parameter {name=None, typeDefHandle {registeredTypeIndex=1}, varData []}}
      SamplingRegions []
     }}
   )
   &ParticleState()
   &SkeletalMeshLocation(
    Bone Sampling Mode=::NewEnumerator4
    Inherit Velocity=false
    MaskSpawnGroup=false
    Sampled Position Offset:NiagaraPosition=$Particles.Initial.Position
    Skeletal Mesh=$User.Skeletal Mesh Root
   )
   /SetVariables(Particles.MeshOrientation=&MultiplyQuaternion(
     Quaternion A=&FindQuatBetweenTwoVectors(
      From Vector=<1 0 0>
      To Vector=&NormalizeVector(Vector To Normalize=&Subtract_Vector(
        A=&MakeCustomVectorFromBool(
         Bool=&SetBoolByFloatComparison(A=$Particles.Normalized Spawn Group, B=0.5, Comparison Type=::NewEnumerator4)
         True Vector=&Subtract_Vector(A=Linked() {linkedVariable {name=Particles.Position, type=NiagaraPosition}}, B=<0 1 0>)
         False Vector=Linked() {linkedVariable {name=Particles.Position, type=NiagaraPosition}}
        )
        B=$Output.SkeletalMeshLocation.SampledPosition
       ))
     )
     Quaternion B=$Particles.Initial.MeshOrientation
    )) {isSetParameters=true, module=SetVariables_#4}
   &UpdateMeshOrientation(
    Rotation Coordinate Space=::NewEnumerator3
    Rotation Vector=&Lerp_Vector(A=<0 0 0>, B=<0 -2 0>, Alpha=$StackContext.FE DivingCurve)
   )
   &self:NoiseScale(
    Noise Frequency=7
    Noise Speed:NiagaraFloat=&Lerp_Float(A=1.5, B=3.5, Alpha=&Ceil_Float(Float=$Particles.FE DivingCurve))
    Scale Size=0.3
   ) {module=NoiseScale}
   /SetVariables(Particles.Position=&Add_Vector(
     A=Linked() {linkedVariable {name=Particles.Position, type=NiagaraPosition}}
     B=&Lerp_Vector(A=<0 7 -2>, B=<0 7 -17>, Alpha=$Particles.FE DivingCurve)
    )) {isSetParameters=true, module=SetVariables_#5}
   &Color(
    Color=&ColorFromCurve(
     ColorCurve=DataInterface() {propertyValues {
       AlphaCurve {keys [<at 0 1 RCIM_Cubic>]}
       BlueCurve {keys [<at 0.014 0.15 RCIM_Cubic>,<at 0.0758 0.2>,<at 0.1399 1>,<at 0.3533 1>,<at 0.4057 1>,<at 0.4431 1>,<at 0.4909 0.15>,<at 0.5363 0.15>,<at 0.5876 0.2>,<at 0.6366 1>,<at 0.888 1>,<at 0.9174 1>,<at 0.9527 1>,<at 1 0.15>]}
       GreenCurve {keys [<at 0.014 0.15 RCIM_Cubic>,<at 0.0758 0.5>,<at 0.1399 1>,<at 0.3533 0.6667>,<at 0.4057 0.2>,<at 0.4431 0.2>,<at 0.4909 0.15>,<at 0.5363 0.15>,<at 0.5876 0.5>,<at 0.6366 1>,<at 0.888 0.6667>,<at 0.9174 0.2>,<at 0.9527 0.2>,<at 1 0.15>]}
       RedCurve {keys [<at 0.014 0.3 RCIM_Cubic>,<at 0.0758 0.05>,<at 0.1399 0>,<at 0.3533 0>,<at 0.4057 0.3333>,<at 0.4431 0.2>,<at 0.4909 0.3>,<at 0.5363 0.3>,<at 0.5876 0.05>,<at 0.6366 0>,<at 0.888 0>,<at 0.9174 0.3333>,<at 0.9527 0.2>,<at 1 0.3>]}
      }}
     CurveIndex=$Particles.Normalized Spawn Group
    )
    Scale Color=&VectorFromFloat(Value=&FN_FrontEndFloatModifier(
      Gameplay=&Multiply_Float(A=&FNTimeOfDayScalar(Night Scale=0.6, Morning Scale=0.3, Evening Scale=0.3), B=0.2)
      Front End=0.4
     ))
   )
   &ScaleColor(
    Color Value To Scale=$Particles.Color
    Curve Index=&Multiply_Float(A=$Particles.NormalizedAge, B=0.05)
    Linear Color Curve=DataInterface() {propertyValues {
      AlphaCurve {keys [<at 0 1>]}
      BlueCurve {keys [<at 0 0.444>,<at 0.2641 0.96>,<at 0.4666 0.36>,<at 0.7161 1.2>,<at 1 0.444>]}
      GreenCurve {keys [<at 0 0.36>,<at 0.2641 1.2>,<at 0.4666 0.5467>,<at 0.7161 1.0133>,<at 1 0.36>]}
      RedCurve {keys [<at 0 1.2>,<at 0.2641 0.12>,<at 0.4666 1.2>,<at 0.7161 0.36>,<at 1 1.2>]}
     }}
    Scale Mode=::NewEnumerator2
   )
  ]
  rendererClasses [&NiagaraMeshRendererProperties]
  renderers [{rendererClass=&NiagaraMeshRendererProperties, rendererIndex=0}]
 }
 Mesh_Ring {
  emitterUpdateScript [
   &EmitterState(
    Inactive Response=::NewEnumerator1
    Life Cycle Mode=::NewEnumerator1
    Loop Behavior=::NewEnumerator1
    Loop Duration Mode=::NewEnumerator1
    UseLoopDelay=false
   )
   &SpawnBurst_Instantaneous(
    Spawn Count=&MakeCustomIntFromBool(Bool=$User.bIsFrontEndPreview, True INT=0, False INT=1)
    Use Spawn Probability=false
   )
  ]
  particleSpawnScript [
   &InitializeParticle(Color=<rgba 1 1 1 0.25>, Mesh Scale=<0.52 0.52 0.5>, Mesh Scale Mode=::NewEnumerator4, UsePositionOffset=false, Write Material Random=true)
   &InitialMeshOrientation(Rotation=<0 0 0.25>, Use Orientation Vector=true, Use Rotation Vector=true)
  ]
  particleUpdateScript [
   &FN_DivingAttributes_v2(
    Accumulate Dive=1
    Accumulate Rest=0
    Bone Index=0
    SkeletalMesh=DataInterface() {propertyValues {
      DefaultMesh=null
      ExcludeBoneName=root
      OptionalComponentClass=null
      PreviewMesh=&M_Med_Soldier_04_Preview
      SkinningMode=SkinOnTheFly
      SoftSourceActor=null
      SourceMode=Default
      UvSetIndex=0
      WholeMeshLOD=-1
      bExcludeBone=false
      bReadDeformedGeometry=true
      bRequireCurrentFrameData=true
      ComponentTags []
      FilteredBones [spine_02]
      FilteredSockets []
      MeshUserParameter {parameter {name=None, typeDefHandle {registeredTypeIndex=1}, varData []}}
      SamplingRegions []
     }}
   )
   &ParticleState(Kill Particles When Lifetime Has Elapsed=false, Let Infinitely Lived Particles Die When Emitter Deactivates=true)
   &SkeletalMeshLocation(
    Bone Sampling Mode=::NewEnumerator4
    Inherit Velocity=false
    MaskSpawnGroup=false
    Sampled Position Offset=&Lerp_Vector(
     A=&FN_FrontEndVector3Modifier(Gameplay=<0 0 -8>, Front End=<0 0 -12>)
     B=<0 0 -12>
     Alpha=$Particles.Dive
    )
    Skeletal Mesh=$User.Skeletal Mesh Root
   )
   &ScaleMeshSize(Scale Factor=&Lerp_Vector(A=<1 1 1>, B=<1 1 1>, Alpha=$Particles.Dive))
   &ScaleColor(
    Scale Alpha=&FN_FrontEndFloatModifier(Gameplay=1, Front End=0.5)
    Scale RGB=&Add_Vector(
     A=<0.1 0.1 0.1>
     B=&Multiply_VectorByFloat(Vector=<1 1 1>, Float=&Multiply_Float(A=3, B=$System.Audio NPC 2))
    )
    ScaleA=true
    ScaleRGB=true
   )
   &DynamicMaterialParameters(
    Index 0 Param 1=&Lerp_Float(A=1, B=2, Alpha=&Ceil_Float(Float=$Particles.Dive))
    Index 0 Param 2=1
    Module.Param0WriteEnabled=true
    Module.Param1WriteEnabled=true
    Module.Param2WriteEnabled=true
    Module.Param3WriteEnabled=true
   )
  ]
  rendererClasses [&NiagaraMeshRendererProperties]
  renderers [{rendererClass=&NiagaraMeshRendererProperties, rendererIndex=0}]
 }
 Mesh_Ring_FE {
  bEnabled=false
  emitterUpdateScript [
   &EmitterState(Life Cycle Mode=::NewEnumerator1, Loop Duration=20, UseLoopDelay=false)
   &SpawnBurst_Instantaneous(
    Spawn Count=&MakeCustomIntFromBool(Bool=$User.bIsFrontEndPreview, True INT=1, False INT=0)
    Use Spawn Probability=false
   )
  ]
  particleSpawnScript [
   &InitializeParticle(Color=<rgba 1 1 1 0.25>, Mesh Scale=<0.55 0.55 0.5>, Mesh Scale Mode=::NewEnumerator4, UsePositionOffset=false, Write Material Random=true)
   &InitialMeshOrientation(Rotation=<0 0 0.25>, Use Orientation Vector=true, Use Rotation Vector=true)
  ]
  particleUpdateScript [
   /SetVariables(Particles.FE DivingCurve:NiagaraFloat=&FloatFromCurve(
     FloatCurve=DataInterface() {propertyValues {Curve {keys [<at 0 0 RCIM_Cubic>, {arriveTangent=2.2332, leaveTangent=2.2332, time=0.01, value=1}, <at 0.5 1 RCIM_Cubic>, <at 0.51 0>, <at 1 0>]}}}
     CurveIndex=$Emitter.NormalizedLoopAge
    )) {isSetParameters=true}
   &FN_DivingAttributes_v2(
    Accumulate Dive=1
    Accumulate Rest=0
    Bone Index=0
    SkeletalMesh=DataInterface() {propertyValues {
      DefaultMesh=null
      ExcludeBoneName=root
      OptionalComponentClass=null
      PreviewMesh=&M_Med_Soldier_04_Preview
      SkinningMode=SkinOnTheFly
      SoftSourceActor=null
      SourceMode=Default
      UvSetIndex=0
      WholeMeshLOD=-1
      bExcludeBone=false
      bReadDeformedGeometry=true
      bRequireCurrentFrameData=true
      ComponentTags []
      FilteredBones [spine_02]
      FilteredSockets []
      MeshUserParameter {parameter {name=None, typeDefHandle {registeredTypeIndex=1}, varData []}}
      SamplingRegions []
     }}
   )
   &ParticleState(Kill Particles When Lifetime Has Elapsed=false, Let Infinitely Lived Particles Die When Emitter Deactivates=true)
   &SkeletalMeshLocation(
    Bone Sampling Mode=::NewEnumerator4
    Inherit Velocity=false
    MaskSpawnGroup=false
    Sampled Position Offset=&Lerp_Vector(A=<0 7 -15>, B=<0 7 -27>, Alpha=$Particles.FE DivingCurve)
    Skeletal Mesh=$User.Skeletal Mesh Root
   )
   &ScaleMeshSize(Scale Factor=&Lerp_Vector(A=<1 1 1>, B=<1 1 1>, Alpha=$Particles.Dive))
   &ScaleColor(Scale Alpha=&FN_FrontEndFloatModifier(Gameplay=1, Front End=0.5), ScaleA=true, ScaleRGB=true)
   &UpdateMeshOrientation(Rotation Rate=-0.2, Rotation Vector=<0 0 1>)
  ]
  rendererClasses [&NiagaraMeshRendererProperties]
  renderers [{rendererClass=&NiagaraMeshRendererProperties, rendererIndex=0}]
 }
 Particle_Attach {
  emitterUpdateScript [
   &EmitterState(Life Cycle Mode=::NewEnumerator1, Loop Behavior=::NewEnumerator1, Loop Duration Mode=::NewEnumerator1, UseLoopDelay=false)
   &SpawnBurst_Instantaneous(Spawn Count=4, Use Spawn Probability=false)
  ]
  particleSpawnScript [
   &FN_SetSpawnGroup()
   &InitializeParticle(
    AdjustAlpha=true
    AdjustHue=true
    AdjustSaturation=true
    AdjustValue=true
    Alpha Scale Range:Vector2f=<0.1 0.1>
    Color=<rgba 0 0.5 1 1>
    Color Mode=::NewEnumerator3
    Hue Shift Range:Vector2f=<-0.02 0.01>
    Saturation Range:Vector2f=<1 1>
    Sprite Size=&FN_FrontEndVector2Modifier(Front End:Vector2f=<30 90>, Gameplay:Vector2f=<50 70>)
    Sprite Size Mode=::NewEnumerator4
    UsePositionOffset=false
    Value Range:Vector2f=<0.8 0.5>
    Write Material Random=true
   )
   /Niagara/Modules/Spawn/Location/SphereLocation(MaskSpawnGroup=false, Module.Override Randomness=false, Sphere Radius=500, Surface Only=false)
  ]
  particleUpdateScript [
   &ParticleState(Kill Particles When Lifetime Has Elapsed=false, Let Infinitely Lived Particles Die When Emitter Deactivates=true)
   &SkeletalMeshLocation(
    Bone / Socket Index=$Particles.SpawnGroup
    Bone Sampling Mode=::NewEnumerator4
    Inherit Velocity=false
    Interpolate Mesh Sampling=false
    MaskSpawnGroup=false
    Sampled Position Offset=<0 0 6>
    Skeletal Mesh=DataInterface() {propertyValues {
      DefaultMesh=null
      ExcludeBoneName=root
      OptionalComponentClass=null
      PreviewMesh=&M_Med_Soldier_04_Preview
      SkinningMode=SkinOnTheFly
      SoftSourceActor=null
      SourceMode=Default
      UvSetIndex=0
      WholeMeshLOD=-1
      bExcludeBone=true
      bReadDeformedGeometry=true
      bRequireCurrentFrameData=true
      ComponentTags []
      FilteredBones [hand_l,hand_r,foot_l,foot_r,upperarm_l,upperarm_r]
      FilteredSockets []
      MeshUserParameter {parameter {name=None, typeDefHandle {registeredTypeIndex=1}, varData []}}
      SamplingRegions []
     }}
   )
   &CameraOffset(Camera Offset Amount=30)
   &ScaleSpriteSize(Scale Factor=&Vector2DFromFloat(Value=&SelectFloatFromArray(Direct Array Index=$Particles.SpawnGroup, Float Selection Array=DataInterface() {propertyValues {FloatData [0.7,0.7,1,1,1,1]}})))
   &DynamicMaterialParameters(
    Index 0 Param 1=1
    Index 0 Param 2=1
    Index 0 Param 3=&RandomRangeFloat(Minimum=1.5, Maximum=1)
    Index 0 Param 4=&RandomRangeFloat(Minimum=0.7, Maximum=1)
    Module.Param0WriteEnabled=true
    Module.Param1WriteEnabled=true
    Module.Param2WriteEnabled=true
    Module.Param3WriteEnabled=true
   )
   /NiagaraFN/Sprite/FN_SpriteAlignmentAndFacing(SpriteAlignment:Vector3f=<0 0 1>)
   &ScaleColor(
    Scale RGB=&VectorFromFloat(Value=&Lerp_Float(A=1, B=4, Alpha=$System.Audio NPC 2))
    ScaleA=true
    ScaleRGB=true
   )
   &ScaleColor(
    Scale Alpha=&Multiply_Float(
     A=&FNTimeOfDayScalar(Night Scale=0.6, Morning Scale=0.3, Evening Scale=0.3)
     B=&FN_FrontEndFloatModifier(Gameplay=1, Front End=1.3)
    )
    ScaleA=true
    ScaleRGB=true
   ) {enabled=false, module=ScaleColor001}
  ]
  rendererClasses [&NiagaraSpriteRendererProperties]
  renderers [{rendererClass=&NiagaraSpriteRendererProperties}]
 }
 Particle_Box {
  emitterUpdateScript [
   &EmitterState()
   &SpawnRate(
    SpawnRate=&Multiply_Float(A=12, B=&Lerp_Float(A=0, B=10, Alpha=&Power(a=$System.Audio NPC 2, b=2)))
    Use Spawn Probability=false
   )
  ]
  particleSpawnScript [
   &InitializeParticle(
    AdjustAlpha=true
    AdjustHue=true
    AdjustSaturation=false
    AdjustValue=true
    Alpha Scale Range:Vector2f=<1 0.5>
    Color=<rgba 0.408 0.016 0.8 1>
    Color Mode=::NewEnumerator3
    Lifetime Max=0.5
    Lifetime Min=0.4
    Lifetime Mode=::NewEnumerator1
    Mesh Scale Max=<1.5 0.25 0.5>
    Mesh Scale Min=<3 0.2 0.3>
    Mesh Scale Mode=::NewEnumerator2
    UsePositionOffset=false
    Value Range:Vector2f=<1 2>
    Write Material Random=true
   )
   &FN_SetSpawnGroup()
   &SkeletalMeshLocation(
    Bone / Socket Index=$Particles.SpawnGroup
    Bone Sampling Mode=::NewEnumerator4
    Inherit Velocity=false
    MaskSpawnGroup=false
    Skeletal Mesh=$User.Skeletal Mesh
   )
   &InitialMeshOrientation(Rotation=<0 0.25 0>, Use Orientation Vector=true, Use Rotation Vector=true)
   &AddVelocity(Velocity=&RandomRangeVector(Minimum=<0 0 600>, Maximum=<0 0 200>))
  ]
  particleUpdateScript [
   &ParticleState()
   &SolveForcesAndVelocity(Clamp Velocity=false, Limit Acceleration=false)
   &ScaleColor(
    Linear Color Curve=DataInterface() {propertyValues {
      AlphaCurve {keys [<at 0.7 1>,<at 1 0>]}
      BlueCurve {keys [<at 0 1>,<at 1 0.5>]}
      GreenCurve {keys [<at 0 1>,<at 1 0.5>]}
      RedCurve {keys [<at 0 1>,<at 1 0.5>]}
     }}
    Scale Mode=::NewEnumerator2
   )
   &ScaleColor(
    Color Value To Scale=$Particles.Color
    Scale RGB=&VectorFromFloat(Value=&Lerp_Float(A=0.5, B=4, Alpha=$System.Audio NPC 2))
    ScaleA=true
    ScaleRGB=true
   ) {module=ScaleColor001}
   &ScaleMeshSize(Scale Factor=&VectorFromFloat(Value=&Lerp_Float(A=0.1, B=1, Alpha=$System.Audio NPC 2))) {enabled=false}
  ]
  rendererClasses [&NiagaraMeshRendererProperties]
  renderers [{rendererClass=&NiagaraMeshRendererProperties, rendererIndex=0}]
 }
 Particle_Decal {
  emitterUpdateScript [
   &EmitterState(Life Cycle Mode=::NewEnumerator1, Loop Duration=20, UseLoopDelay=false)
   &SpawnBurst_Instantaneous(Spawn Count=4, Use Spawn Probability=false)
  ]
  particleSpawnScript [
   &FN_SetSpawnGroup()
   &InitializeParticle(
    Color=<rgba 0.3 1.65 3 1>
    Sprite Size Mode=::NewEnumerator3
    Uniform Sprite Size=&SelectFloatFromArray(
     Direct Array Index=&SpawnGroup()
     Float Selection Array=DataInterface() {propertyValues {FloatData [50,50,100,100]}}
    )
    UsePositionOffset=false
    Write Material Random=true
   )
  ]
  particleUpdateScript [
   /SetVariables(Particles.FE DivingCurve:NiagaraFloat=&FloatFromCurve(
     FloatCurve=DataInterface() {propertyValues {Curve {keys [<at 0 0 RCIM_Cubic>, {arriveTangent=2.2332, leaveTangent=2.2332, time=0.01, value=1}, <at 0.5 1 RCIM_Cubic>, <at 0.51 0>, <at 1 0>]}}}
     CurveIndex=$Emitter.NormalizedLoopAge
    )) {isSetParameters=true}
   &FN_DivingAttributes_v2(Accumulate Dive=1, Accumulate Rest=0, Bone Index=0, SkeletalMesh:NiagaraDataInterfaceSkeletalMesh=$User.Skeletal Mesh Root)
   &ParticleState(Kill Particles When Lifetime Has Elapsed=false, Let Infinitely Lived Particles Die When Emitter Deactivates=true)
   &CameraOffset(Camera Offset Amount=&SelectFloatFromArray(
     Direct Array Index=&SpawnGroup()
     Float Selection Array=DataInterface() {propertyValues {FloatData [10,10,20,20]}}
    ))
   &SkeletalMeshLocation(
    Apply Sampled Orientation to Mesh Particles=true
    Bone / Socket Index=$Particles.SpawnGroup
    Bone Sampling Mode=::NewEnumerator4
    Inherit Velocity=false
    MaskSpawnGroup=false
    Sampled Position Offset=<0 0 10>
    Sampled Position Offset Coordinate Space:ENiagaraMeshTransforms=::NewEnumerator1
    Skeletal Mesh=$User.Skeletal Mesh
   )
   &ScaleColor(
    Scale RGB=&VectorFromFloat(Value=&Lerp_Float(A=0.5, B=2, Alpha=$Particles.Dive))
    ScaleA=true
    ScaleRGB=true
   )
   &ScaleColor(
    Scale Alpha=&Multiply_Float(
     A=&FNTimeOfDayScalar(Night Scale=0.5, Morning Scale=0.2, Evening Scale=0.3)
     B=&FN_FrontEndFloatModifier(Gameplay=1, Front End=&Lerp_Float(A=1, B=3, Alpha=$Particles.FE DivingCurve))
    )
    ScaleA=true
    ScaleRGB=true
   ) {enabled=false, module=ScaleColor001}
  ]
  rendererClasses [&NiagaraSpriteRendererProperties]
  renderers [{rendererClass=&NiagaraSpriteRendererProperties}]
 }
 Trail_Mesh {
  emitterUpdateScript [
   &EmitterState(
    Inactive Response=::NewEnumerator1
    Life Cycle Mode=::NewEnumerator1
    Loop Behavior=::NewEnumerator1
    Loop Duration Mode=::NewEnumerator1
    UseLoopDelay=false
   )
   &SpawnBurst_Instantaneous(Spawn Count=4, Use Spawn Probability=false)
  ]
  particleSpawnScript [
   &FN_SetSpawnGroup()
   &InitializeParticle(
    Mesh Scale=&SelectVectorFromArray(
     Direct Array Index=$Particles.SpawnGroup
     Vector Selection Array=DataInterface() {propertyValues {FloatData [<1.6 1.6 2.2>, <1.6 1.6 2.2>, <2.2 2.2 2>, <2.2 2.2 2>]}}
    )
    Mesh Scale Mode=::NewEnumerator4
    UsePositionOffset=false
    Write Material Random=true
   )
   &InitialMeshOrientation(
    Orientation Axis=<0 0 1>
    Orientation Vector=<0 0 1>
    Rotation=&RandomRangeVector(Minimum=<0 0 -1>, Maximum=<0 0 1>)
    Rotation Coordinate Space=::NewEnumerator0
    Use Orientation Vector=true
    Use Rotation Vector=true
   )
   &CameraOffset(Camera Offset Amount=10)
  ]
  particleUpdateScript [
   &FN_DivingAttributes_v2(Accumulate Dive=1, Accumulate Rest=0, Bone Index=0, SkeletalMesh:NiagaraDataInterfaceSkeletalMesh=$User.Skeletal Mesh Root)
   &ParticleState(Kill Particles When Lifetime Has Elapsed=false, Let Infinitely Lived Particles Die When Emitter Deactivates=true)
   &SkeletalMeshLocation(
    Bone / Socket Index=$Particles.SpawnGroup
    Bone Sampling Mode=::NewEnumerator4
    Inherit Velocity=false
    MaskSpawnGroup=false
    Sampled Position Offset=<0 0 -5>
    Skeletal Mesh=$User.Skeletal Mesh
   )
   &ScaleMeshSize(Scale Factor=&Lerp_Vector(
     A=&FN_FrontEndVector3Modifier(Gameplay=<1 1 0.85>, Front End=<1.3 1.3 0.85>)
     B=<1 1 1.5>
     Alpha=$Particles.Dive
    ))
   &ScaleMeshSize(
    Initial Mesh Scale:Vector3f=$Particles.Scale
    Scale Factor=&VectorFromFloat(Value=&Lerp_Float(A=0.5, B=1.5, Alpha=$System.Audio NPC 2))
   ) {module=ScaleMeshSize001}
   &DynamicMaterialParameters(
    Index 0 Param 3=&FN_FrontEndFloatModifier(Gameplay=1, Front End=0.35)
    Index 0 Param 4=&RandomRangeFloat(Minimum=0.8, Maximum=1.5)
    Module.Param0WriteEnabled=true
    Module.Param1WriteEnabled=true
    Module.Param2WriteEnabled=true
    Module.Param3WriteEnabled=true
   )
   &ScaleColor(
    Scale RGB=&VectorFromFloat(Value=&Lerp_Float(A=0.5, B=4, Alpha=$System.Audio NPC 2))
    ScaleA=true
    ScaleRGB=true
   )
   &ScaleColor(
    Scale Alpha=&Multiply_Float(
     A=&FNTimeOfDayScalar(Night Scale=0.6, Morning Scale=0.3, Evening Scale=0.3)
     B=&FN_FrontEndFloatModifier(Gameplay=1, Front End=2)
    )
    Scale RGB=&VectorFromFloat(Value=&FN_FrontEndFloatModifier(Gameplay=1, Front End=1))
    ScaleA=true
    ScaleRGB=true
   ) {enabled=false, module=ScaleColor001}
  ]
  rendererClasses [&NiagaraMeshRendererProperties]
  renderers [{rendererClass=&NiagaraMeshRendererProperties, rendererIndex=0}]
 }
 Wave_BoneFlow_Arm {
  emitterUpdateScript [
   &EmitterState(Life Cycle Mode=::NewEnumerator1, Loop Duration=7.766279436516721e+18, UseLoopDelay=false)
   &SpawnBurst_Instantaneous(Spawn Count=40, Use Spawn Probability=false)
  ]
  particleSpawnScript [
   &FN_SetSpawnGroup(Inclusivity=true, Normalize Spawn Group=true, Particles Per Group=20, Pattern=::NewEnumerator0, Set Ribbon ID and Ribbon Link Order=true) {module=FN_SetSpawnGroup001}
   &InitializeRibbon(
    Module.Write Color=true
    Module.Write Lifetime=true
    Module.Write Mass=true
    Module.Write Material Random=true
    Module.Write Position=false
    Module.Write RibbonTwist=false
    Module.Write RibbonWidth=true
    Ribbon Width=&FloatFromCurve(
     FloatCurve=DataInterface() {propertyValues {Curve {keys [
         {arriveTangent=17.3279, interpMode=RCIM_Cubic, leaveTangent=17.3279, tangentMode=RCTM_User, time=0, value=1}
         {interpMode=RCIM_Cubic, tangentMode=RCTM_User, time=0.5, value=5}
         {arriveTangent=-20.1583, interpMode=RCIM_Cubic, leaveTangent=-20.1583, tangentMode=RCTM_User, time=1, value=1}
        ]}}}
     CurveIndex=&ReturnNormalizedExecIndex()
    )
   )
  ]
  particleUpdateScript [
   &ParticleState(Kill Particles When Lifetime Has Elapsed=false, Let Infinitely Lived Particles Die When Emitter Deactivates=true)
   /Game/Effects/Niagara/Modules/BoneFlow/FN_BoneFlow_v4/FN_BoneFlow(
    Angle:NiagaraFloat=&Multiply_Float(A=$Particles.Normalized Spawn Group, B=0)
    BoneID:NiagaraInt32=&ModuloInt(A=$Particles.SpawnGroup, B=2)
    Height:NiagaraFloat=$Particles.Normalized Spawn Group
    Interpolate Mesh Sampling=false
    Radius:NiagaraFloat=&Multiply_Float(
     A=&FloatFromCurve(
      FloatCurve=DataInterface() {propertyValues {Curve {keys [
          {interpMode=RCIM_Cubic, tangentMode=RCTM_User, time=0, value=0}
          {interpMode=RCIM_Cubic, tangentMode=RCTM_User, time=0.5, value=1}
          {tangentMode=RCTM_User, time=1, value=0}
         ]}}}
      CurveIndex=$Particles.Normalized Spawn Group
      Scale Curve=15
     )
     B=$System.Audio NPC 2
    )
    SideVector1:Vector3f=<0 0 0>
    SideVector2:Vector3f=&MaskVectorBySpawnGroup(Masked Vector=<0 -1 2>, Passthrough Vector=<0 1 -1>)
    SideVector3:Vector3f=<0 0 0>
    SkeletalMesh=DataInterface() {propertyValues {
      DefaultMesh=null
      ExcludeBoneName=root
      OptionalComponentClass=null
      PreviewMesh=&M_Med_Soldier_04_Preview
      SkinningMode=SkinOnTheFly
      SoftSourceActor=null
      SourceMode=Default
      UvSetIndex=0
      WholeMeshLOD=-1
      bExcludeBone=false
      bReadDeformedGeometry=true
      bRequireCurrentFrameData=true
      ComponentTags []
      FilteredBones [lowerarm_twist_02_r, hand_r, middle_03_r, lowerarm_twist_02_l, hand_l, middle_03_l]
      FilteredSockets []
      MeshUserParameter {parameter {name=None, typeDefHandle {registeredTypeIndex=1}, varData []}}
      SamplingRegions []
     }}
    Stabalize Twisting=true
    Write Position Attribute=true
   ) {module=FN_BoneFlow001}
   &DynamicMaterialParameters(
    Index 0 Param 1=1
    Index 0 Param 2=1
    Index 0 Param 3=1
    Index 0 Param 4=1
    Module.Param0WriteEnabled=true
    Module.Param1WriteEnabled=true
    Module.Param2WriteEnabled=true
    Module.Param3WriteEnabled=true
   )
   &CameraOffset(Camera Offset Amount=20)
  ]
  rendererClasses [&NiagaraRibbonRendererProperties]
  renderers [{rendererClass=&NiagaraRibbonRendererProperties}]
 }
 Wave_BoneFlow_Back {
  emitterUpdateScript [
   &EmitterState(Life Cycle Mode=::NewEnumerator1, UseLoopDelay=false)
   &SpawnBurst_Instantaneous(Spawn Count=40, Use Spawn Probability=false)
  ]
  particleSpawnScript [
   &FN_SetSpawnGroup(Inclusivity=true, Normalize Spawn Group=true, Particles Per Group=20, Pattern=::NewEnumerator0, Set Ribbon ID and Ribbon Link Order=true) {module=FN_SetSpawnGroup001}
   &InitializeRibbon(
    Module.Write Color=true
    Module.Write Lifetime=true
    Module.Write Mass=true
    Module.Write Material Random=true
    Module.Write Position=false
    Module.Write RibbonTwist=false
    Module.Write RibbonWidth=true
    Ribbon Width=&FloatFromCurve(
     FloatCurve=DataInterface() {propertyValues {Curve {keys [
         {arriveTangent=17.3279, interpMode=RCIM_Cubic, leaveTangent=17.3279, tangentMode=RCTM_User, time=0, value=1}
         {interpMode=RCIM_Cubic, tangentMode=RCTM_User, time=0.5, value=5}
         {arriveTangent=-20.1583, interpMode=RCIM_Cubic, leaveTangent=-20.1583, tangentMode=RCTM_User, time=1, value=1}
        ]}}}
     CurveIndex=&ReturnNormalizedExecIndex()
     Scale Curve=2
    )
   )
  ]
  particleUpdateScript [
   &ParticleState(Kill Particles When Lifetime Has Elapsed=false, Let Infinitely Lived Particles Die When Emitter Deactivates=true)
   &FN_BoneFlow(
    Angle:NiagaraFloat=&Multiply_Float(A=$Particles.Normalized Spawn Group, B=&MaskFloatBySpawnGroup(Masked Float=0, Passthrough Float=0))
    BoneID:NiagaraInt32=&ModuloInt(A=$Particles.SpawnGroup, B=2)
    Height:NiagaraFloat=$Particles.Normalized Spawn Group
    Interpolate Mesh Sampling=false
    Radius:NiagaraFloat=&Multiply_Float(
     A=&FloatFromCurve(
      FloatCurve=DataInterface() {propertyValues {Curve {keys [
          {interpMode=RCIM_Cubic, tangentMode=RCTM_User, time=0, value=0}
          {interpMode=RCIM_Cubic, tangentMode=RCTM_User, time=0.5, value=1}
          {tangentMode=RCTM_User, time=1, value=0}
         ]}}}
      CurveIndex=$Particles.Normalized Spawn Group
     )
     B=&Lerp_Float(A=0, B=15, Alpha=$System.Audio NPC 2)
    )
    SideVector1:Vector3f=<0 0 0>
    SideVector2:Vector3f=&MaskVectorBySpawnGroup(Masked Vector=<0 1 -0.5>, Passthrough Vector=<0 -1 0.5>)
    SideVector3:Vector3f=<0 0 0>
    SkeletalMesh=DataInterface() {propertyValues {
      DefaultMesh=null
      ExcludeBoneName=root
      OptionalComponentClass=null
      PreviewMesh=&M_Med_Soldier_04_Preview
      SkinningMode=SkinOnTheFly
      SoftSourceActor=null
      SourceMode=Default
      UvSetIndex=0
      WholeMeshLOD=-1
      bExcludeBone=false
      bReadDeformedGeometry=true
      bRequireCurrentFrameData=true
      ComponentTags []
      FilteredBones [lowerarm_twist_02_r, upperarm_twist_02_r, upperarm_r, lowerarm_twist_02_l, upperarm_twist_02_l, upperarm_l]
      FilteredSockets []
      MeshUserParameter {parameter {name=None, typeDefHandle {registeredTypeIndex=1}, varData []}}
      SamplingRegions []
     }}
    Stabalize Twisting=true
    Write Position Attribute=true
   ) {module=FN_BoneFlow001}
   &CameraOffset(Camera Offset Amount=30)
   &DynamicMaterialParameters(
    Index 0 Param 1=0.5
    Index 0 Param 4=-1
    Module.Param0WriteEnabled=true
    Module.Param1WriteEnabled=true
    Module.Param2WriteEnabled=true
    Module.Param3WriteEnabled=true
   )
  ]
  rendererClasses [&NiagaraRibbonRendererProperties]
  renderers [{rendererClass=&NiagaraRibbonRendererProperties}]
 }
 Wave_BoneFlow_Back001 {
  emitterUpdateScript [
   &EmitterState(Inactive Response=::NewEnumerator1, Life Cycle Mode=::NewEnumerator1, UseLoopDelay=false)
   &SpawnBurst_Instantaneous(Spawn Count=20, Use Spawn Probability=false)
  ]
  particleSpawnScript [
   &FN_SetSpawnGroup(Inclusivity=true, Normalize Spawn Group=true, Particles Per Group=20, Pattern=::NewEnumerator0, Set Ribbon ID and Ribbon Link Order=true) {module=FN_SetSpawnGroup001}
   &InitializeRibbon(
    Module.Write Color=true
    Module.Write Lifetime=true
    Module.Write Mass=true
    Module.Write Material Random=true
    Module.Write Position=false
    Module.Write RibbonTwist=false
    Module.Write RibbonWidth=true
    Ribbon Width=&FloatFromCurve(
     FloatCurve=DataInterface() {propertyValues {Curve {keys [
         {arriveTangent=17.3279, interpMode=RCIM_Cubic, leaveTangent=17.3279, tangentMode=RCTM_User, time=0, value=1}
         {interpMode=RCIM_Cubic, tangentMode=RCTM_User, time=0.5, value=5}
         {arriveTangent=-20.1583, interpMode=RCIM_Cubic, leaveTangent=-20.1583, tangentMode=RCTM_User, time=1, value=1}
        ]}}}
     CurveIndex=&ReturnNormalizedExecIndex()
     Scale Curve=3
    )
   )
  ]
  particleUpdateScript [
   &ParticleState(
    Kill Particles When Lifetime Has Elapsed=false
    Let Infinitely Lived Particles Die When Emitter Deactivates=true
    Loop Particles Lifetime=true
   )
   &FN_BoneFlow(
    Angle:NiagaraFloat=&Multiply_Float(A=$Particles.Normalized Spawn Group, B=&MaskFloatBySpawnGroup(Masked Float=0, Passthrough Float=0))
    BoneID:NiagaraInt32=&ModuloInt(A=$Particles.SpawnGroup, B=1)
    Height:NiagaraFloat=$Particles.Normalized Spawn Group
    Interpolate Mesh Sampling=false
    Radius:NiagaraFloat=&Multiply_Float(
     A=&FloatFromCurve(
      FloatCurve=DataInterface() {propertyValues {Curve {keys [
          {interpMode=RCIM_Cubic, tangentMode=RCTM_User, time=0, value=0}
          {interpMode=RCIM_Cubic, tangentMode=RCTM_User, time=0.5, value=1}
          {tangentMode=RCTM_User, time=1, value=0}
         ]}}}
      CurveIndex=$Particles.Normalized Spawn Group
     )
     B=&Lerp_Float(A=0, B=50, Alpha=$System.Audio NPC 2)
    )
    SideVector1:Vector3f=<0 0 0>
    SideVector2:Vector3f=<0 -1 0>
    SideVector3:Vector3f=<0 0 0>
    SkeletalMesh=DataInterface() {propertyValues {
      DefaultMesh=null
      ExcludeBoneName=root
      OptionalComponentClass=null
      PreviewMesh=&M_Med_Soldier_04_Preview
      SkinningMode=SkinOnTheFly
      SoftSourceActor=null
      SourceMode=Default
      UvSetIndex=0
      WholeMeshLOD=-1
      bExcludeBone=false
      bReadDeformedGeometry=true
      bRequireCurrentFrameData=true
      ComponentTags []
      FilteredBones [upperarm_twist_02_l, neck_02, upperarm_twist_02_r]
      FilteredSockets []
      MeshUserParameter {parameter {name=None, typeDefHandle {registeredTypeIndex=1}, varData []}}
      SamplingRegions []
     }}
    Stabalize Twisting=true
    Write Position Attribute=true
   ) {module=FN_BoneFlow001}
   &CameraOffset(Camera Offset Amount=30)
   &DynamicMaterialParameters(
    Index 0 Param 1=1
    Index 0 Param 2=&FloatFromCurve(FloatCurve=DataInterface() {propertyValues {Curve {keys [<at 0 0.9>,<at 0.07 1.125>,<at 0.17 0.9>,<at 0.4 0.9>,<at 0.5 1.375>,<at 0.6 0.9>,<at 0.675 1.125>,<at 0.75 0.9>,<at 1 0.9>]}}})
    Index 0 Param 4=-0.85
    Module.Param0WriteEnabled=true
    Module.Param1WriteEnabled=true
    Module.Param2WriteEnabled=true
    Module.Param3WriteEnabled=true
   )
  ]
  rendererClasses [&NiagaraRibbonRendererProperties]
  renderers [{rendererClass=&NiagaraRibbonRendererProperties}]
 }
 Wave_BoneFlow_Leg {
  emitterUpdateScript [
   &EmitterState(Life Cycle Mode=::NewEnumerator1, Loop Duration=7.766279436516721e+18, UseLoopDelay=false)
   &SpawnBurst_Instantaneous(Spawn Count=20, Use Spawn Probability=false)
  ]
  particleSpawnScript [
   &FN_SetSpawnGroup(Inclusivity=true, Normalize Spawn Group=true, Particles Per Group=10, Pattern=::NewEnumerator0, Set Ribbon ID and Ribbon Link Order=true) {module=FN_SetSpawnGroup001}
   &InitializeRibbon(
    Module.Write Color=true
    Module.Write Lifetime=true
    Module.Write Mass=true
    Module.Write Material Random=true
    Module.Write Position=false
    Module.Write RibbonTwist=false
    Module.Write RibbonWidth=true
    Ribbon Width=&FloatFromCurve(
     FloatCurve=DataInterface() {propertyValues {Curve {keys [
         {arriveTangent=17.3279, interpMode=RCIM_Cubic, leaveTangent=17.3279, tangentMode=RCTM_User, time=0, value=1}
         {interpMode=RCIM_Cubic, tangentMode=RCTM_User, time=0.5, value=5}
         {arriveTangent=-20.1583, interpMode=RCIM_Cubic, leaveTangent=-20.1583, tangentMode=RCTM_User, time=1, value=1}
        ]}}}
     CurveIndex=&ReturnNormalizedExecIndex()
    )
   )
  ]
  particleUpdateScript [
   &ParticleState(Kill Particles When Lifetime Has Elapsed=false, Let Infinitely Lived Particles Die When Emitter Deactivates=true)
   &FN_BoneFlow(
    Angle:NiagaraFloat=&Multiply_Float(A=$Particles.Normalized Spawn Group, B=&MaskFloatBySpawnGroup(Masked Float=0.2, Passthrough Float=-0.2))
    BoneID:NiagaraInt32=&ModuloInt(A=$Particles.SpawnGroup, B=2)
    Height:NiagaraFloat=$Particles.Normalized Spawn Group
    Interpolate Mesh Sampling=false
    Radius:NiagaraFloat=&Multiply_Float(
     A=&FloatFromCurve(
      FloatCurve=DataInterface() {propertyValues {Curve {keys [
          {interpMode=RCIM_Cubic, tangentMode=RCTM_User, time=0, value=0}
          {interpMode=RCIM_Cubic, tangentMode=RCTM_User, time=0.5, value=1}
          {tangentMode=RCTM_User, time=1, value=0}
         ]}}}
      CurveIndex=$Particles.Normalized Spawn Group
      Scale Curve=26
     )
     B=$System.Audio NPC 2
    )
    SideVector1:Vector3f=<0 0 0>
    SideVector2:Vector3f=&MaskVectorBySpawnGroup(Masked Vector=<0 1 0.2>, Passthrough Vector=<0 -1 -0.2>)
    SideVector3:Vector3f=<0 0 0>
    SkeletalMesh=DataInterface() {propertyValues {
      DefaultMesh=null
      ExcludeBoneName=root
      OptionalComponentClass=null
      PreviewMesh=&M_Med_Soldier_04_Preview
      SkinningMode=SkinOnTheFly
      SoftSourceActor=null
      SourceMode=Default
      UvSetIndex=0
      WholeMeshLOD=-1
      bExcludeBone=false
      bReadDeformedGeometry=true
      bRequireCurrentFrameData=true
      ComponentTags []
      FilteredBones [calf_r, calf_twist_01_r, ball_r, calf_l, calf_twist_01_l, ball_l]
      FilteredSockets []
      MeshUserParameter {parameter {name=None, typeDefHandle {registeredTypeIndex=1}, varData []}}
      SamplingRegions []
     }}
    Stabalize Twisting=true
    Write Position Attribute=true
   ) {module=FN_BoneFlow001}
   &CameraOffset(Camera Offset Amount=20)
   &DynamicMaterialParameters(
    Index 0 Param 1=1
    Index 0 Param 2=1
    Index 0 Param 3=1
    Index 0 Param 4=1
    Module.Param0WriteEnabled=true
    Module.Param1WriteEnabled=true
    Module.Param2WriteEnabled=true
    Module.Param3WriteEnabled=true
   )
  ]
  rendererClasses [&NiagaraRibbonRendererProperties]
  renderers [{rendererClass=&NiagaraRibbonRendererProperties}]
 }
 Wave_Mesh {
  emitterUpdateScript [
   &EmitterState(
    Inactive Response=::NewEnumerator1
    Life Cycle Mode=::NewEnumerator1
    Loop Behavior=::NewEnumerator1
    Loop Duration Mode=::NewEnumerator1
    UseLoopDelay=false
   )
   &SpawnBurst_Instantaneous(Spawn Count=4, Use Spawn Probability=false)
  ]
  particleSpawnScript [
   &FN_SetSpawnGroup()
   &InitializeParticle(
    Mesh Scale=&SelectVectorFromArray(
     Direct Array Index=&SpawnGroup()
     Vector Selection Array=DataInterface() {propertyValues {FloatData [<0.32 0.32 0.25>, <0.32 0.32 0.25>, <0.43 0.43 0.35>, <0.43 0.43 0.35>, <0.25 0.25 0.2>, <0.25 0.25 0.2>]}}
    )
    Mesh Scale Mode=::NewEnumerator4
    UsePositionOffset=false
    Write Material Random=true
   )
   &InitialMeshOrientation(
    Orientation Axis=<0 0 1>
    Orientation Vector=<0 0 1>
    Rotation Coordinate Space=::NewEnumerator0
    Use Orientation Vector=true
    Use Rotation Vector=true
   )
  ]
  particleUpdateScript [
   &FN_DivingAttributes_v2(
    Accumulate Dive=0.1
    Accumulate Rest=-0.1
    Bone Index=0
    SkeletalMesh=DataInterface() {propertyValues {
      DefaultMesh=null
      ExcludeBoneName=root
      OptionalComponentClass=null
      PreviewMesh=&M_Med_Soldier_04_Preview
      SkinningMode=SkinOnTheFly
      SoftSourceActor=null
      SourceMode=Default
      UvSetIndex=0
      WholeMeshLOD=-1
      bExcludeBone=false
      bReadDeformedGeometry=true
      bRequireCurrentFrameData=true
      ComponentTags []
      FilteredBones [pelvis]
      FilteredSockets []
      MeshUserParameter {parameter {name=None, typeDefHandle {registeredTypeIndex=1}, varData []}}
      SamplingRegions []
     }}
   )
   &ParticleState(
    Kill Particles When Lifetime Has Elapsed=false
    Let Infinitely Lived Particles Die When Emitter Deactivates=true
    Loop Particles Lifetime=true
   )
   &SkeletalMeshLocation(
    Bone / Socket Index=$Particles.SpawnGroup
    Bone Sampling Mode=::NewEnumerator4
    Inherit Velocity=false
    MaskSpawnGroup=false
    Sampled Position Offset=&FN_FrontEndVector3Modifier(
     Gameplay=&SelectVectorFromArray(
      Direct Array Index=$Particles.SpawnGroup
      Vector Selection Array=DataInterface() {propertyValues {FloatData [<13 0 0>, <-13 0 0>, <-10 12 0>, <10 -12 0>]}}
     )
     Front End=&SelectVectorFromArray(
      Direct Array Index=$Particles.SpawnGroup
      Vector Selection Array=DataInterface() {propertyValues {FloatData [<10 0 0>, <-10 0 0>, <-4 12 2>, <4 -12 -2>]}}
     )
    )
    Sampled Position Offset Coordinate Space:ENiagaraMeshTransforms=::NewEnumerator3
    Skeletal Mesh=$User.Skeletal Mesh
   )
   &ScaleMeshSize(Scale Factor=&MakeVector(X=1, Y=1, Z=&Lerp_Float(A=0.3, B=1, Alpha=$System.Audio NPC 2))) {module=ScaleMeshSize001}
   &ScaleMeshSize(Scale Factor=&Lerp_Vector(A=<1 1 1>, B=<1 1 6>, Alpha=$Particles.Dive))
   &ScaleColor(
    Scale Alpha=&Multiply_Float(
     A=&FNTimeOfDayScalar(Night Scale=0.6, Morning Scale=0.2, Day Scale=0.88, Evening Scale=0.3)
     B=&FN_FrontEndFloatModifier(Gameplay=1, Front End=1)
    )
    ScaleA=true
    ScaleRGB=true
   ) {module=ScaleColor001}
   &DynamicMaterialParameters(
    Index 0 Param 1=1
    Index 0 Param 2=$System.Audio NPC 2
    Index 0 Param 4=&MaskFloatBySpawnGroup(
     Masked Float=&RandomRangeFloat(Minimum=-1, Maximum=1)
     Passthrough Float=&RandomRangeFloat(Minimum=-1, Maximum=1)
    )
    Module.Param0WriteEnabled=true
    Module.Param1WriteEnabled=true
    Module.Param2WriteEnabled=true
    Module.Param3WriteEnabled=true
   )
   &CameraOffset(Camera Offset Amount=10)
  ]
  rendererClasses [&NiagaraMeshRendererProperties]
  renderers [{rendererClass=&NiagaraMeshRendererProperties, rendererIndex=0}]
 }
}
summary {
 systemName=NS_Contrail_EnergyNet_Proto
 emitters [
  {emitterName=Wave_Mesh, rendererClasses [&NiagaraMeshRendererProperties]}
  {emitterName=Trail_Mesh, rendererClasses [&NiagaraMeshRendererProperties]}
  {emitterName=Wave_BoneFlow_Arm, rendererClasses [&NiagaraRibbonRendererProperties]}
  {emitterName=Wave_BoneFlow_Leg, rendererClasses [&NiagaraRibbonRendererProperties]}
  {emitterName=Wave_BoneFlow_Back, rendererClasses [&NiagaraRibbonRendererProperties]}
  {emitterName=Wave_BoneFlow_Back001, rendererClasses [&NiagaraRibbonRendererProperties]}
  {emitterName=Mesh_Box, rendererClasses [&NiagaraMeshRendererProperties]}
  {emitterName=Mesh_Ring, rendererClasses [&NiagaraMeshRendererProperties]}
  {emitterName=Particle_Box, rendererClasses [&NiagaraMeshRendererProperties]}
  {emitterName=Particle_Decal, rendererClasses [&NiagaraSpriteRendererProperties]}
  {emitterName=Particle_Attach, rendererClasses [&NiagaraSpriteRendererProperties]}
  {bEnabled=false, emitterName=Mesh_Box_FE, rendererClasses [&NiagaraMeshRendererProperties]}
  {bEnabled=false, emitterName=Mesh_Ring_FE, rendererClasses [&NiagaraMeshRendererProperties]}
  {bEnabled=false, emitterName=Debug, rendererClasses [&NiagaraSpriteRendererProperties]}
 ]
 userVariables [
  User.AuidoPreview {type=NiagaraFloat, defaultValue=0.5, description=""}
  User.bIsFrontEnd {type=NiagaraBool}
  User.bIsFrontEndPreview {type=NiagaraBool}
  User.isReactive {type=NiagaraFloat, defaultValue=0, description=""}
  User.PlayerSpeed {type=NiagaraFloat, defaultValue=0, description=""}
  User.Skeletal Mesh {
   type=NiagaraDataInterfaceSkeletalMesh
   defaultValue {
    struct=NiagaraExt_VariableValue_DataInterface
    value {dataInterface=&self:NiagaraDataInterfaceSkeletalMesh_0}
   }
  }
  User.Skeletal Mesh Root {
   type=NiagaraDataInterfaceSkeletalMesh
   defaultValue {
    struct=NiagaraExt_VariableValue_DataInterface
    value {dataInterface=&self:NiagaraDataInterfaceSkeletalMesh_3}
   }
  }
  User.Wave {type=NiagaraFloat, defaultValue=1, description=""}
 ]
}
